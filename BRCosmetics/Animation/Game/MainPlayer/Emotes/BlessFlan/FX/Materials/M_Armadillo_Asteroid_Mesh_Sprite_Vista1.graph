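# Masked, lit material that draws flipbook rock sprites for asteroids in the BlessFlan emote effects, as camera-facing mesh sprites or particles via the UsedWithParticles switch.
# In mesh mode, vertex colors pick the frame and scale each sprite, and sine-based World Position Offset adds drift.
# Particle mode takes frame and tint from particle data, dropping sprite offset and drift.
# View-axis rotation spins position, plus normals except on ES3_1.
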Material /BRCosmetics/Animation/Game/MainPlayer/Emotes/BlessFlan/FX/Materials/M_Armadillo_Asteroid_Mesh_Sprite_Vista1
graph {
 properties {
  bTangentSpaceNormal=false
  bUseMaterialAttributes=true
  bUsedWithNiagaraSprites=true
  bUsedWithSplineMeshes=true
  blendMode=BLEND_Masked
  pixelDepthOffsetMode=PDOM_Legacy
  usedShadingModels=Default Lit
 }
 nodes [
  CameraVectorWS()
  n1 = VertexColor()
  n2 = Power(Base=@n1.R, Exp=ScalarParameter() {defaultValue=3, parameterName=SizeExponent}) {constExponent=3}
  LinearInterpolate(Alpha=@n2) {constA=0.3}
  n3 = VertexColor()
  Multiply(A=Subtract(A=@n3) {constB=0.5}) {constB=2}
  n4 = ParticleColor()
  n5 = VectorParameter() {parameterName=BaseColor, defaultValue=<rgba 0.0397 0.0459 0.05 1>}
  n6 = Constant() {r=16}
  n7 = Round(None=Multiply(A=VertexColor().G, B=@n6))
  n8 = Constant() {r=4}
  n9 = MaterialFunctionCall(
   "Animation  Phase (0-1)"=StaticSwitchParameter(True=DynamicParameter() {paramNames [Random 0-1, Param2, Param3, Param4]}.Random 0-1, False=Divide(A=@n7, B=@n6)) {parameterName=UsedWithParticles}
   Number of Rows=@n8
   Number of Columns=@n8
  ) {materialFunction=&FlipBook}
  n10 = TextureSample(UVs=@n9.UVs) {texture=/Game/Effects/Fort_Effects/Textures/Rocks/T_Rock_Stone_Sprite_D}
  n11 = WorldPosition()
  n12 = ScalarParameter() {defaultValue=50, parameterName=SizeMult}
  n13 = LinearInterpolate(A=ScalarParameter() {defaultValue=50, parameterName=SizeMultLOW}, B=@n12, Alpha=@n2)
  n14 = ScalarParameter() {defaultValue=50, parameterName=Size_Min}
  n15 = TextureSample(UVs=@n9.UVs) {samplerType=SAMPLERTYPE_Normal, texture=/Game/Effects/Fort_Effects/Textures/Rocks/T_Rock_Stone_Sprite_N}
  n16 = MaterialFunctionCall!(
   Center Location=@n11.XYZ
   0-1 UVs
   XY scale=StaticSwitchParameter(True=Max(A=@n13, B=@n14), False=Multiply(A=@n12, B=@n1.R)) {parameterName=Use MinMaxScale}
   OverridePivot
   "Normals (Optional)"=@n15.RGB
  ) {materialFunction=&Sprite}
  n17 = Transform(None=StaticSwitchParameter(True=@n15.RGB, False=@n16.Normals) {parameterName=UsedWithParticles})
  n18 = Normalize(VectorInput=Subtract(A=@n11.XYZ, B=CameraPositionWS()))
  n19 = VertexColor()
  n20 = Time()
  n21 = ScalarParameter() {parameterName=RandomRotationSpeed}
  n22 = Add(A=Multiply(A=@n19.G, B=Multiply(A=@n20, B=@n21)), B=Multiply(A=ScalarParameter() {parameterName=RandomRotation}, B=@n19.R))
  n23 = RotateAboutAxis(NormalizedRotationAxis=@n18, RotationAngle=@n22, PivotPoint=Constant3Vector(), Position=@n17)
  n24 = LinearInterpolate(Alpha=@n1.R) {constA=1, constB=0.5}
  n25 = Time()
  n26 = ScalarParameter() {parameterName=WPODriftSpeed}
  n27 = MaterialFunctionCall(Float3=Add(A=@n3, B=Multiply(A=@n25, B=@n26))) {materialFunction=&BreakOutFloat3Components}
  n28 = Multiply(
   A=MaterialFunctionCall(R=Sine(None=@n27.R) {period=0.97}, G=Sine(None=@n27.G) {period=1.16}, B=Sine(None=@n27.B) {period=0.8}) {materialFunction=&AppendMany}.RGB
   B=Multiply(A=ScalarParameter() {parameterName=WPODriftMult}, B=ScalarParameter() {defaultValue=50, parameterName=SizeMult})
  )
  n29 = StaticSwitchParameter(True=Constant(), False=Add(A=@n16.World position offset, B=Multiply(A=@n24, B=@n28))) {parameterName=UsedWithParticles}
  n30 = Add(A=@n29, B=WorldPosition().XYZ)
  SetMaterialAttributes!(
   MaterialAttributes
   Base Color=Multiply(A=StaticSwitchParameter(True=@n4.RGB, False=@n5.RGB) {parameterName=UsedWithParticles}, B=LinearInterpolate(Alpha=@n10.G) {constA=0.6})
   Roughness=LinearInterpolate(Alpha=@n10.G) {constA=1, constB=0.8}
   Emissive Color
   Opacity Mask=@n10.A
   Normal=FeatureLevelSwitch(Default=Add(A=@n17, B=@n23), ES3_1=@n17)
   World Position Offset=Add(A=@n29, B=RotateAboutAxis(NormalizedRotationAxis=@n18, RotationAngle=@n22, PivotPoint=@n11.XYZ, Position=@n30))
  ) {attributeSetTypes [
    #1
    #2
    #3
    #4
    #5
    #6
   ]}
 ]
}
