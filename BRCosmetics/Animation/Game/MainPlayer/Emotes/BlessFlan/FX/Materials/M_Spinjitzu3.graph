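# Translucent unlit material for the Spinjitzu tornado effect in the BlessFlan emote, usable on particle, Niagara and skeletal meshes.
# Panning noise masks blend dark, main, highlight and intersection colors, tinted by particle color.
# Opacity math combines noise-modulated edge masks, view-angle fade, particle alpha and a gradient.
# World position offset pushes vertices along normals using noise and contraction masks, switchable to spline thickening, with optional dynamic-parameter offset; dynamic parameters also modulate masks.
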Material /BRCosmetics/Animation/Game/MainPlayer/Emotes/BlessFlan/FX/Materials/M_Spinjitzu3
graph {
 properties {
  bAutomaticallySetUsageInEditor=false
  bUsedWithMeshDeformer=true
  bUsedWithMeshParticles=true
  bUsedWithNiagaraMeshParticles=true
  bUsedWithNiagaraRibbons=true
  bUsedWithSkeletalMesh=true
  bUsedWithSplineMeshes=true
  blendMode=BLEND_Translucent
  shadingModel=MSM_Unlit
  usedShadingModels=Unlit
 }
 nodes [
  n1 = ParticleColor()
  n2 = VectorParameter() {group=Opacity, parameterName=Opacity_EdgeMask_InvertControl, channelNames {a="", b=Power, g=Multiply, r=Offset}, defaultValue=<rgba 0 1 1 0>}
  n3 = VectorParameter() {group=Opacity, parameterName=Opacity_EdgeMask_Control, channelNames {a="", b=Power, g=Multiply, r=Offset}, defaultValue=<rgba 0 5.2389 1 0>}
  n4 = DynamicParameter() {parameterIndex=1, defaultValue=<rgba 1 0.4801 40 1>, paramNames [0,power,grad,Param4]}
  n5 = TextureObjectParameter() {
   group=Highlight
   parameterName=Distortion_Texture
   samplerType=SAMPLERTYPE_LinearGrayscale
   sortPriority=0
   texture=/Game/Effects/Fort_Effects/Textures/Tile/T_Diagonal_Blur_Noise_M
  }
  n6 = VectorParameter() {group=Highlight, parameterName=Highlight_UVControl, channelNames {a=VPan, b=UPan, g=VScale, r=UScale}, defaultValue=<rgba 1 2 -1.4 0.4>}
  n7 = MaterialFunctionCall!(Texture=@n5, AdditionalUVOffset, UVControl_RGB=@n6, UVControl_A=@n6, "[Override]TimeInput", "[Override]ScaleUVByCenter", "[Override]UVInput") {materialFunction=/Game/Effects/Niagara/MasterMaterials/Functions/Common/FXCommon_TextureSample_V2}
  n8 = MaterialFunctionCall!(
   UorV=ComponentMask(None=TextureCoordinate()) {g=true}
   Invert_EdgeControl=@n2
   NoInvert_EdgeControl=Multiply(A=@n3, B=Multiply(A=@n4.RGB, B=@n7.RGB))
  ) {materialFunction=/Game/Effects/Niagara/MasterMaterials/Functions/Common/FXCommon_EdgeMask_Ex}
  n9 = Multiply(A=NamedRerouteUsage(), B=@n8.Mask)
  n10 = ScalarParameter() {defaultValue=1, parameterName="FadeSensitivity (optional) (S)"}
  n11 = Multiply(
   A=Multiply(
    A=LinearInterpolate(A=@n9, B=@n8.Mask, Alpha=@n8.Mask)
    B=MaterialFunctionCall("FadeSensitivity (optional)"=@n10) {materialFunction=&FadeBasedOnViewAngle}.Result
   ) {constA=1}
   B=@n1.A
  )
  n12 = DynamicParameter() {parameterIndex=3, defaultValue=<rgba 1 0.4801 1 1>, paramNames [0,power,grad,Param4]}
  n13 = MaterialFunctionCall(UV Channel=TextureCoordinate()) {materialFunction=&LinearGradient}
  n14 = Add(A=@n12.grad, B=@n13.VGradient)
  Add(A=@n11, B=@n14)
  n15 = TextureObjectParameter() {
   group=MainTexture
   parameterName=MainTexture_Texture
   samplerType=SAMPLERTYPE_LinearGrayscale
   sortPriority=0
   texture=&T_Diagonal_Blur_Noise_M
  }
  n16 = VectorParameter() {group=MainTexture, parameterName=MainTexture_UVControl, channelNames {a=VPan, b=UPan, g=VScale, r=UScale}, defaultValue=<rgba 1 2 -1 1>}
  n17 = VectorParameter() {group=MainTexture, parameterName=MainTexture_Refine, channelNames {a=Lerp, b=Base, g=Mult, r=Power}, defaultValue=<rgba 9.7206 100 1 0.5>}
  n18 = Saturate(None=MaterialFunctionCall(
    Texture=MaterialFunctionCall(Texture=@n15, UVControl_RGB=@n16, UVControl_A=@n16) {materialFunction=&FXCommon_TextureSample_V2}.G
    RGB=@n17
    A=@n17
   ) {materialFunction=&MF_FX_Refine}.Result)
  n19 = TextureObjectParameter() {
   group=Highlight
   parameterName=Distortion_Texture
   samplerType=SAMPLERTYPE_LinearGrayscale
   sortPriority=0
   texture=&T_Diagonal_Blur_Noise_M
  }
  n20 = VectorParameter() {group=Highlight, parameterName=Highlight_UVControl, channelNames {a=VPan, b=UPan, g=VScale, r=UScale}, defaultValue=<rgba 1 2 -1.4 0.4>}
  n21 = VectorParameter() {group=Highlight, parameterName=Highlight_Refine, channelNames {a=Lerp, b=Base, g=Mult, r=Power}, defaultValue=<rgba 7 0.5 0.5 0>}
  n22 = Saturate(None=MaterialFunctionCall(
    Texture=MaterialFunctionCall(Texture=@n19, UVControl_RGB=@n20, UVControl_A=@n20) {materialFunction=&FXCommon_TextureSample_V2}.G
    RGB=@n21
    A=@n21
   ) {materialFunction=&MF_FX_Refine}.Result)
  Max(A=@n18, B=@n22)
  n23 = LinearInterpolate(
   A=LinearInterpolate(
    A=VectorParameter() {parameterName=Color_Dark, defaultValue=<rgba 0.4 0.1979 0.0014 1>}.RGB
    B=VectorParameter() {parameterName=Color_Main, defaultValue=<rgba 1.5 0.5792 0.0068 1>}.RGB
    Alpha=@n18
   )
   B=VectorParameter() {parameterName=Color_Highlight, defaultValue=<rgba 8 3.0572 0.4813 1>}.RGB
   Alpha=@n22
  )
  n24 = VectorParameter() {parameterName=Color_Intersection, defaultValue=<rgba 7 2.675 0.4211 1>}
  n25 = ComponentMask(None=TextureCoordinate()) {g=true}
  n26 = VectorParameter() {
   group=IntersectionMask
   parameterName=IntersectionMask_EdgeMask_InvertControl
   channelNames {a=null, b=Power, g=Multiply, r=Offset}
   defaultValue=<rgba 0 1 0 0>
  }
  n27 = VectorParameter() {
   group=IntersectionMask
   parameterName=IntersectionMask_EdgeMask_Control
   channelNames {a=null, b=Power, g=Multiply, r=Offset}
   defaultValue=<rgba 0 1 15 0>
  }
  n28 = Saturate(None=MaterialFunctionCall(UorV=@n25, Invert_EdgeControl=@n26, NoInvert_EdgeControl=@n27) {materialFunction=&FXCommon_EdgeMask_Ex}.Mask)
  Multiply(
   A=MaterialFunctionCall(Emissive=LinearInterpolate(A=@n23, B=@n24.RGB, Alpha=@n28)) {materialFunction=&MF_FX_EmissiveAdjustments}.Result
   B=@n1.RGB
  )
  Multiply(A=@n13.VGradient, B=@n12.grad)
  Multiply(A=@n11, B=@n14)
  NamedRerouteDeclaration(None=@n18) {name=GreyscaleMask, nodeColor=<rgba 1 0 0.375 1>}
  n29 = ScalarParameter() {defaultValue=50, group=Spline, parameterName=SplineWidth}
  n30 = ScalarParameter() {defaultValue=-40, desc=max size tornado can reach inward, group=WPO, parameterName=WPO_Min}
  n31 = ScalarParameter() {defaultValue=50, desc=Maximum size tornado can reach outward, group=WPO, parameterName=WPO_Max}
  n32 = TextureObjectParameter() {
   group=WPO
   parameterName=WPO_Texture
   samplerType=SAMPLERTYPE_Grayscale
   sortPriority=0
   texture=&LowResBlurredNoise
  }
  n33 = VectorParameter() {group=WPO, parameterName=WPO_UVControl, channelNames {a=VPan, b=UPan, g=VScale, r=UScale}, defaultValue=<rgba 1 1 1 0>}
  n34 = MaterialFunctionCall(Texture=@n32, UVControl_RGB=@n33, UVControl_A=@n33) {materialFunction=&FXCommon_TextureSample_V2}
  n35 = TextureCoordinate()
  n36 = ComponentMask(None=@n35) {g=true}
  n37 = VectorParameter() {
   group=WPOContraction
   parameterName=WPOContraction_EdgeMask_InvertControl
   channelNames {a=null, b=Power, g=Multiply, r=Offset}
   defaultValue=<rgba 0 1 1 0>
  }
  n38 = VectorParameter() {
   group=WPOContraction
   parameterName=WPOContraction_EdgeMask_Control
   channelNames {a=null, b=Power, g=Multiply, r=Offset}
   defaultValue=<rgba -0.104 1.9083 1 0>
  }
  n39 = Add(
   A=Multiply(
    A=MaterialFunctionCall(UorV=@n36, Invert_EdgeControl=@n37, NoInvert_EdgeControl=@n38) {materialFunction=&FXCommon_EdgeMask_Ex}.Mask
    B=ScalarParameter() {defaultValue=-290.8049, group=WPOContraction, parameterName=WPOContraction_Amount}
   )
   B=Add(A=LinearInterpolate(A=@n30, B=@n31, Alpha=@n34.G), B=ScalarParameter() {defaultValue=184.9625, group=WPO, parameterName=WPO Intensity})
  )
  n40 = VertexNormalWS()
  n41 = OneMinus(None=@n25)
  n42 = VectorParameter() {group=WPO, parameterName=WPO_EdgeMask_InvertControl, channelNames {a="", b=Power, g=Multiply, r=Offset}, defaultValue=<rgba 0 0.3739 0.2 0>}
  n43 = Multiply(A=@n42, B=DynamicParameter() {defaultValue=<rgba 1 0.4801 40 1>, paramNames [0,power,grad,Param4]}.RGB)
  n44 = VectorParameter() {group=WPO, parameterName=WPO_EdgeMask_Control, channelNames {a="", b=Power, g=Multiply, r=Offset}, defaultValue=<rgba 0.5 1 3.5511 0.5>}
  n45 = StaticSwitchParameter(
   True=MaterialFunctionCall(WidthBase=@n29, WidthTip=@n29, Expand U or V UV Channel=StaticBoolParameter() {group=Spline, parameterName="Expand U or V UV Channel (SB)"}) {materialFunction=&SplineThicken}.WorldPositionOffset
   False=Multiply(
    A=Multiply(A=@n39, B=@n40) {constB=0}
    B=MaterialFunctionCall(UorV=@n41, Invert_EdgeControl=@n43, NoInvert_EdgeControl=@n44) {materialFunction=&FXCommon_EdgeMask_Ex}.Mask
   )
  ) {group=Spline, parameterName=UseSplineThicken}
  n46 = DynamicParameter() {parameterIndex=1, defaultValue=<rgba 1 0.4801 40 1>, paramNames [0,power,grad,Param4]}
  n47 = OneMinus(None=ComponentMask(None=TextureCoordinate()) {g=true})
  n48 = VectorParameter() {
   group=WPO
   parameterName=WPOInterpolate_EdgeMask_InvertControl
   channelNames {a=null, b=Power, g=Multiply, r=Offset}
   defaultValue=<rgba 0 1 1 0>
  }
  n49 = VectorParameter() {group=WPO, parameterName=WPOInterpolate_EdgeMask_Control, channelNames {a=null, b=Power, g=Multiply, r=Offset}, defaultValue=<rgba 0 1 1 0>}
  n50 = Add(
   A=@n45
   B=Multiply(
    A=MaterialFunctionCall(R=@n46.0, G=@n46.0, B=@n46.0) {materialFunction=&AppendMany}.RGB
    B=MaterialFunctionCall(UorV=@n47, Invert_EdgeControl=@n48, NoInvert_EdgeControl=@n49) {materialFunction=&FXCommon_EdgeMask_Ex}.Mask
   )
  )
  Reroute(None=Multiply(A=StaticSwitchParameter(True=@n50, False=@n45) {defaultValue=true, parameterName=UseInterpolateWPO}))
  Sine(None=@n35)
  TextureSample() {samplerType=SAMPLERTYPE_LinearGrayscale, texture=&T_Diagonal_Blur_Noise_M}
 ]
}
