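# Niagara system for the fire tornado effect of the BlessFlan emote.
# Tornado, ring, base and glow mesh emitters combine with vortex-driven fire sprites, GPU sparks and flickering lights, each on its own timed loop.
# Ribbon flame trails spawn from the fire sprite particles, while flash sprites attach to leg and foot bones.
# Curves drive fades and dynamic material parameters, and ring timing comes from user parameters.
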NiagaraSystem /BRCosmetics/Animation/Game/MainPlayer/Emotes/BlessFlan/FX/NS_Emote_BlessFlan_FireTornado1
dependencies {
 usedDataInterfaces [
  &NiagaraDataInterfaceArrayFloat
  &NiagaraDataInterfaceSkeletalMesh
  &NiagaraDataInterfaceArrayFloat3
  &NiagaraDataInterfaceParticleRead
  &NiagaraDataInterfaceSpriteRendererInfo
  &NiagaraDataInterfaceArrayColor
  &NiagaraDataInterfaceVectorCurve
 ]
 usedDynamicInputs [
  &Multiply_LinearColorByFloat
  &MakeFloatFromLinearColor
  &SimulationPosition
  &SetBoolByFloatComparison
  &VectorLength
 ]
 usedModules [
  &SystemState
  &Timeline
  &self:SystemSpawnScript.SystemScriptSource.SystemScriptGraph.NiagaraNodeAssignment_6.SetVariables_#1
  &self:SystemSpawnScript.SystemScriptSource.SystemScriptGraph.NiagaraNodeAssignment_0.SetVariables_#2
 ]
}
emitters {
 NE_BottomBase {
  emitterUpdateScript [
   &EmitterState(Life Cycle Mode=::NewEnumerator1, Loop Behavior=::NewEnumerator1, Loop Duration=4.8, UseLoopDelay=false)
   &SpawnRate(SpawnRate=15, Use Spawn Probability=false)
  ]
  particleSpawnScript [
   &InitializeParticle(
    Color Mode=::NewEnumerator0
    Lifetime=0.5
    Mesh Scale Mode=::NewEnumerator3
    Mesh Uniform Scale=2.5
    Position Offset=&RandomRangeVector(Minimum=<0 0 1>, Maximum=<1 1 3>)
    UsePositionOffset=true
    Write Material Random=false
   ) {module=InitializeParticle001}
   &InitialMeshOrientation(
    Mesh Orientation Mode=::NewEnumerator7
    Rotation=&RandomRangeVector(Minimum=<0 0 0>, Maximum=<0 0 360>)
    Use Rotation Vector=true
   )
  ]
  particleUpdateScript [
   &ParticleState()
   &UpdateMeshOrientation(Rotation Rate=-2, Rotation Vector=<0 0 1>)
   &ScaleMeshSize(Scale Factor=&VectorFromFloat(Value=&FloatFromCurve(FloatCurve=DataInterface() {propertyValues {Curve {
         defaultValue=1
         keys [{arriveTangent=1.226, interpMode=RCIM_Cubic, leaveTangent=1.226, tangentMode=RCTM_User, time=0, value=0.2}, <at 1 1 RCIM_Cubic>]
        }}})))
   &ScaleColor(
    Scale Alpha=&FloatFromCurve(FloatCurve=DataInterface() {propertyValues {Curve {keys [<at 0 1>,<at 1 0>]}}})
    ScaleA=true
    ScaleRGB=true
   )
   &FN_Gen5_Light(
    Attenutation:NiagaraFloat=&FloatFromCurve(
     FloatCurve=DataInterface() {propertyValues {Curve {keys [
         <at 0 1 RCIM_Cubic>
         {arriveTangent=104.2105, interpMode=RCIM_Cubic, leaveTangent=104.2105, time=0.07, value=100}
         <at 0.95 100>
         <at 1.0037 0>
        ]}}}
     Scale Curve=2
    )
    Color:LinearColor=<rgba 1 0.5418 0.0396 1>
    Intensity:NiagaraFloat=&Multiply_Float(
     A=&FloatFromCurve(
      FloatCurve=DataInterface() {propertyValues {Curve {keys [
          <at 0 1 RCIM_Cubic>
          {arriveTangent=3123.75, interpMode=RCIM_Cubic, leaveTangent=3123.75, time=0.07, value=2500}
          <at 0.8 2500>
          <at 1.0037 0>
         ]}}}
      Scale Curve=0.5
     )
     B=/Niagara/DynamicInputs/Multiply/ScaleAndBiasFloat(Float:NiagaraFloat=$System.Flicker, Scale:NiagaraFloat=1, Bias:NiagaraFloat=1)
    )
    Position Offset:Vector3f=<0 0 70>
   )
   &CameraOffset(Camera Offset Amount=15)
  ]
  rendererClasses [&NiagaraMeshRendererProperties, &NiagaraComponentRendererProperties]
  renderers [
   {rendererClass=&NiagaraMeshRendererProperties, rendererIndex=0}
   {rendererClass=&NiagaraComponentRendererProperties, rendererIndex=1}
  ]
 }
 NE_FireSprites {
  emitterUpdateScript [
   &EmitterState(Life Cycle Mode=::NewEnumerator1, Loop Behavior=::NewEnumerator1, Loop Duration=4.85, UseLoopDelay=false)
   &SpawnRate(SpawnRate=20, Use Spawn Probability=false)
  ]
  particleSpawnScript [
   /SetVariables(Particles.Random size:NiagaraFloat=&RandomRangeFloat(Minimum=0.15, Maximum=0.25)) {isSetParameters=true}
   &InitializeParticle(
    Lifetime Max=1
    Lifetime Min=0.6
    Lifetime Mode=::NewEnumerator1
    Sprite Rotation Mode=::NewEnumerator2
    Sprite Size Mode=::NewEnumerator3
    Uniform Sprite Size=&Multiply_Float(A=70, B=$Particles.Random size)
    UsePositionOffset=true
    Write Material Random=true
   )
   &ShapeLocation(
    Handle Radius=0
    Large Radius=20
    MaskSpawnGroup=false
    Non Uniform Scale=<1.5 1.5 1>
    Offset=<0 0 5>
    Offset Mode=::NewEnumerator0
    Override Seed=false
    Shape Primitive=::NewEnumerator3
    UseOwnerScale=true
   ) {enabled=false}
   &ShapeLocation(
    MaskSpawnGroup=false
    Non Uniform Scale=<1.5 1.5 0>
    Offset=<0 0 5>
    Offset Mode=::NewEnumerator0
    Override Seed=false
    Sphere Radius=20
    UseOwnerScale=true
   ) {module=ShapeLocation001}
   &DynamicMaterialParameters(
    Index 0 Param 4=2.5
    Module.Param0WriteEnabled=true
    Module.Param1WriteEnabled=true
    Module.Param2WriteEnabled=false
    Module.Param3WriteEnabled=true
   ) {module=DynamicMaterialParameters001}
   &AddVelocity(
    Cone Angle=150
    Cone Axis=<0 0 1>
    Module.UseConeFalloff=false
    Module.UseVelDistribution=false
    Override Seed=false
    Velocity Mode=::NewEnumerator2
    Velocity Speed=90
   ) {module=AddVelocity001}
  ]
  particleUpdateScript [
   &ParticleState()
   /SetVariables(Particles.Size Scale:NiagaraFloat=&FloatFromCurve(
     FloatCurve=DataInterface() {propertyValues {Curve {keys [
         {interpMode=RCIM_Cubic, tangentMode=RCTM_User, time=0, value=0}
         {interpMode=RCIM_Cubic, tangentMode=RCTM_User, time=0.2, value=0.3}
         {tangentMode=RCTM_User, time=1, value=0.1}
        ]}}}
     Scale Curve=6
    )) {isSetParameters=true}
   &VortexForce(
    Origin Pull Amount=6362.4497
    UseOriginOffset=true
    Vortex Axis=<0 0 -1>
    Vortex Axis Coordinate Space:ENiagaraCoordinateSpace=::Local
    Vortex Force Amount=2500
   )
   &GravityForce(Gravity=&RandomRangeVector(Minimum=<0 0 300>, Maximum=<0 0 400>))
   &SpriteRotationRate(Rotation Rate=&RandomRangeFloat(Minimum=-300, Maximum=300)) {module=SpriteRotationRate001}
   &ScaleSpriteSize(Scale Sprite Size Mode=::NewEnumerator0, Uniform Scale Factor=$Particles.Size Scale) {module=ScaleSpriteSize001}
   &ScaleColor(
    Scale Alpha=&Multiply_Float(
     A=2
     B=&FloatFromCurve(
      FloatCurve=DataInterface() {propertyValues {Curve {keys [<at 0 0 RCIM_Cubic>, {interpMode=RCIM_Cubic, tangentMode=RCTM_User, time=0.5, value=1}, {tangentMode=RCTM_User, time=1, value=0.1}]}}}
      Scale Curve=&FN_FrontEndFloatModifier(Gameplay=1, Front End=0.5)
     )
    )
    Scale RGB=&Multiply_VectorByFloat(
     Vector=&MakeVectorFromLinearColorRGB(Color=&ColorFromCurve(ColorCurve=DataInterface() {propertyValues {
         AlphaCurve {keys []}
         BlueCurve {keys [<at 0.6023 0.1042 RCIM_Cubic>]}
         GreenCurve {keys [<at 0.6023 0.599 RCIM_Cubic>]}
         RedCurve {keys [<at 0.6023 10 RCIM_Cubic>]}
        }}))
     Float=1
    )
    ScaleA=true
    ScaleRGB=true
   )
   &FN_Sprite_Attributes(
    Color=&MakeLinearColorFromVectorAndFloat(
     "Vector (RGB)"=&MakeVectorFromLinearColorRGB(Color=<rgba 1 0.2 0 0.5>)
     "Float (Alpha)"=&Multiply_Float(A=0.33, B=&FN_FrontEndFloatModifier(Gameplay=0.5, Front End=0.2))
    )
    Create Color Attribute=true
    Create SpriteSize Attribute=true
    SpriteSize=&Vector2DFromFloat(Value=&Multiply_Float(A=150, B=&Multiply_Float(A=$Particles.Random size, B=$Particles.Size Scale)))
    Write to Module Namespace=true
   ) {module=FN_Sprite_Attributes001}
   &SubUVAnimation(
    End Frame Range Override=12
    "Sprite Renderer "=DataInterface() {propertyValues {SpriteRenderer=&self:NE_FireSprites_0.NiagaraSpriteRendererProperties_0}}
    Start Frame Range Override=8
    SubUV Lookup Index Scale:NiagaraFloat=&RandomRangeFloat(Minimum=0.2, Maximum=1)
    UseEndFrame=true
    UseStartFrame=true
   )
   &Drag(
    Drag=&FloatFromCurve(FloatCurve=DataInterface() {propertyValues {Curve {defaultValue=1, keys [<at 0 0.5 RCIM_Cubic>,<at 1 3 RCIM_Cubic>]}}})
    Use Linear Drag=true
    Use Rotational Drag=true
   )
   &CurlNoiseForce(Bias Noise=false, Noise Frequency=30, Noise Strength=500, Pan Noise=false) {enabled=false}
   &GravityForce(Gravity=<0 0 122.5>) {enabled=false, module=GravityForce001}
   &SolveForcesAndVelocity(Clamp Velocity=false, Limit Acceleration=false)
   &GenerateLocationEvent(Event Send Rate=90, Use Event Probability=false, UseEventDelay=false)
   &CameraOffset(Camera Offset Amount=15)
  ]
  rendererClasses [&NiagaraSpriteRendererProperties]
  renderers [
   {rendererClass=&NiagaraSpriteRendererProperties}
   {rendererClass=&NiagaraSpriteRendererProperties, rendererIndex=1}
  ]
 }
 NE_Glow_Body {
  emitterUpdateScript [
   /SetVariables(Emitter.GlowOutlineThickness:NiagaraFloat=&FloatFromCurve(
     FloatCurve=DataInterface() {propertyValues {Curve {keys [<at 0 0 RCIM_Cubic>, <at 0.1 1.5>, {arriveTangent=-1.6667, interpMode=RCIM_Cubic, leaveTangent=-1.6667, time=0.8, value=1.5}, <at 1 0>]}}}
     CurveIndex=$System.NormalizedLoopAge
    )) {isSetParameters=true}
   &EmitterState(Life Cycle Mode=::NewEnumerator1, Loop Behavior=::NewEnumerator1, Loop Duration=7.5, UseLoopDelay=false)
   &SpawnBurst_Instantaneous(Use Spawn Probability=false, UseLoopCountLimit=false)
  ]
  particleSpawnScript [&InitializeParticle(
    Color Mode=::NewEnumerator0
    Lifetime=5.5
    Mesh Scale=<1.5 1.5 1.5>
    Mesh Scale Mode=::NewEnumerator4
    Position Offset=<0 0 46>
    UsePositionOffset=true
    Write Material Random=true
   )]
  particleUpdateScript [
   &ParticleState()
   &ScaleColor(
    Scale Alpha=&FloatFromCurve(
     FloatCurve=DataInterface() {propertyValues {Curve {keys [<at 0 0 RCIM_Cubic>, <at 0.1943 1.5>, {arriveTangent=-1.8617, interpMode=RCIM_Cubic, leaveTangent=-1.8617, time=0.8, value=1.5}, <at 1 0>]}}}
     Scale Curve=4
    )
    ScaleA=true
    ScaleRGB=true
   ) {module=ScaleColor001}
  ]
  rendererClasses [&NiagaraMeshRendererProperties]
  renderers [{rendererClass=&NiagaraMeshRendererProperties, rendererIndex=0}]
 }
 NE_ParticleDecalFlash {
  emitterUpdateScript [
   &EmitterState(Life Cycle Mode=::NewEnumerator1, Loop Behavior=::NewEnumerator1, Loop Delay=0.1, Loop Duration=5.5, UseLoopDelay=false)
   &SpawnBurst_Instantaneous(Spawn Count=14, Use Spawn Probability=false, UseLoopCountLimit=false)
  ]
  particleSpawnScript [
   &FN_SetSpawnGroup(Number of Spawn Groups=8)
   &InitializeParticle(
    Color=&RandomRangeLinearColor(Minimum=<rgba 5 0.9583 0.15 1>, Maximum=<rgba 5 0.5542 0.15 1>)
    Lifetime=5
    Sprite Size Mode=::NewEnumerator3
    Sprite UV Mode=::NewEnumerator1
    Uniform Sprite Size=&SelectFloatFromArray(Direct Array Index=$Particles.SpawnGroup, Float Selection Array=DataInterface() {propertyValues {FloatData [50,50,50,50,50,50,50,50,50]}})
    UsePositionOffset=false
    Write Material Random=true
   )
   &CameraOffset(Camera Offset Amount=20)
  ]
  particleUpdateScript [
   &ParticleState()
   &SkeletalMeshLocation(
    Bone / Socket Index=$Particles.SpawnGroup
    Bone / Socket Sampling Mode=::NewEnumerator5
    MaskSpawnGroup=false
    Mesh Sampling Type=::NewEnumerator6
    Sampled Position Offset=&SelectVectorFromArray(
     Direct Array Index=$Particles.SpawnGroup
     Vector Selection Array=DataInterface() {propertyValues {FloatData [<0 0 -50>, <0 0 -50>, <0 0 0>, <0 0 0>, <0 0 0>, <0 0 10>, <0 0 10>]}}
    )
    Skeletal Mesh=DataInterface() {propertyValues {
      DefaultMesh=null
      ExcludeBoneName=root
      OptionalComponentClass=null
      PreviewMesh=/Game/Characters/Player/Male/Male_Avg_Base/M_Med_Base_Skeleton
      SkinningMode=SkinOnTheFly
      SoftSourceActor=null
      SourceMode=Default
      UvSetIndex=0
      WholeMeshLOD=-1
      bExcludeBone=false
      bReadDeformedGeometry=true
      bRequireCurrentFrameData=true
      ComponentTags []
      FilteredBones [calf_r,calf_l,foot_l,foot_r,foot_l,foot_r]
      FilteredSockets [foot_l_toe_tip, foot_r_toe_tip]
      MeshUserParameter {parameter {name=None, typeDefHandle {registeredTypeIndex=1}, varData []}}
      SamplingRegions []
     }}
   )
   &ScaleSpriteSize(Scale Sprite Size Mode=::NewEnumerator2) {module=ScaleSpriteSize001}
   &DynamicMaterialParameters(
    Index 0 Param 1=&FN_FrontEndFloatModifier(
     Gameplay=&FloatFromCurve(
      FloatCurve=DataInterface() {propertyValues {Curve {keys [
          {arriveTangent=-0.2314, interpMode=RCIM_Cubic, leaveTangent=-0.2314, tangentMode=RCTM_User, time=0, value=2}
          {arriveTangent=-3, interpMode=RCIM_Cubic, leaveTangent=-3, time=0.1, value=0}
          {arriveTangent=2.2222, interpMode=RCIM_Cubic, leaveTangent=2.2222, time=0.5, value=0.5}
          {interpMode=RCIM_Cubic, tangentMode=RCTM_User, time=1, value=2}
         ]}}}
      CurveIndex=$Emitter.NormalizedLoopAge
     )
     Front End=&FloatFromCurve(
      FloatCurve=DataInterface() {propertyValues {Curve {keys [
          {arriveTangent=-0.2314, interpMode=RCIM_Cubic, leaveTangent=-0.2314, tangentMode=RCTM_User, time=0, value=2}
          {arriveTangent=-3, interpMode=RCIM_Cubic, leaveTangent=-3, time=0.1, value=0}
          {arriveTangent=2.2222, interpMode=RCIM_Cubic, leaveTangent=2.2222, time=0.5, value=0.5}
          {interpMode=RCIM_Cubic, tangentMode=RCTM_User, time=1, value=2}
         ]}}}
      CurveIndex=$Emitter.NormalizedLoopAge
     )
    )
    Index 0 Param 3=1
    Index 0 Param 4=&RandomRangeFloat(Minimum=0, Maximum=1)
    Module.Param0WriteEnabled=true
    Module.Param1WriteEnabled=true
    Module.Param2WriteEnabled=true
    Module.Param3WriteEnabled=true
   )
   &ScaleColor(
    Scale Alpha=&Multiply_Float(
     A=&FN_FrontEndFloatModifier(Gameplay=1, Front End=1)
     B=&FloatFromCurve(FloatCurve=DataInterface() {propertyValues {Curve {keys [
          {tangentMode=RCTM_User, time=0, value=0}
          <at 0.2 1>
          {arriveTangent=-0.0405, interpMode=RCIM_Cubic, leaveTangent=-0.0405, tangentMode=RCTM_User, time=0.4, value=1}
          {arriveTangent=-1.4489, interpMode=RCIM_Cubic, leaveTangent=-1.4489, tangentMode=RCTM_User, time=1, value=0}
         ]}}})
    )
    Scale RGB=&Multiply_VectorByFloat(
     Vector=&VectorFromFloat(Value=&FNTimeOfDayScalar(Night Scale=1, Morning Scale=0.7, Evening Scale=0.7))
     Float=1
    )
    ScaleA=true
    ScaleRGB=true
   )
   &SkeletalMeshLocation(
    Bone / Socket Index=$Particles.SpawnGroup
    MaskSpawnGroup=false
    Mesh Sampling Type=::NewEnumerator5
    Skeletal Mesh=DataInterface() {propertyValues {
      DefaultMesh=null
      ExcludeBoneName=root
      OptionalComponentClass=null
      PreviewMesh=&M_Med_Base_Skeleton
      SkinningMode=SkinOnTheFly
      SoftSourceActor=null
      SourceMode=Default
      UvSetIndex=0
      WholeMeshLOD=-1
      bExcludeBone=false
      bReadDeformedGeometry=true
      bRequireCurrentFrameData=true
      ComponentTags []
      FilteredBones []
      FilteredSockets [foot_l_toe_tip, foot_r_toe_tip]
      MeshUserParameter {parameter {name=None, typeDefHandle {registeredTypeIndex=1}, varData []}}
      SamplingRegions []
     }}
    Socket Sampling Mode=::NewEnumerator5
   ) {enabled=false, module=SkeletalMeshLocation001}
  ]
  rendererClasses [&NiagaraSpriteRendererProperties]
  renderers [{rendererClass=&NiagaraSpriteRendererProperties}]
 }
 NE_Sparks {
  simTarget=GPUComputeSim
  emitterUpdateScript [
   &EmitterState(Life Cycle Mode=::NewEnumerator1, Loop Behavior=::NewEnumerator1, Loop Duration=4.9, UseLoopDelay=false)
   &SpawnRate(SpawnGroup=2, SpawnRate=70, Use Spawn Probability=false)
  ]
  particleSpawnScript [
   /SetVariables(Particles.VisibilityTag:NiagaraBool=/Niagara/DynamicInputs/Helpers/MaskBoolBySpawnGroup(
     Masked Bool:NiagaraBool=false
     Passthrough Bool:NiagaraBool=true
     Spawn Group Mask:NiagaraInt32=&RandomRangeInt(Minimum=0, Maximum=162351520)
    )) {isSetParameters=true}
   &InitializeParticle(
    Color=<rgba 10 2.1092 0.6029 0.5>
    Lifetime=0.69
    Position Offset=<0 0 40>
    Sprite Size=&Multiply_Vector2D_ByFloat(
     Float=&Multiply_Float(A=1.4, B=&RandomRangeFloat(Minimum=0.3333, Maximum=1))
     Vector2D=<15 50>
    )
    Sprite Size Mode=::NewEnumerator4
    UsePositionOffset=true
    Write Material Random=true
   ) {module=InitializeParticle001}
   &ShapeLocation(
    Handle Radius=20
    Large Radius=40
    MaskSpawnGroup=false
    Offset=<0 0 -40>
    Offset Mode=::NewEnumerator0
    Override Seed=false
    Shape Primitive=::NewEnumerator3
    UseOwnerScale=true
   )
   &AddVelocity(Velocity=<0 0 100>) {enabled=false}
   &AddVelocity(
    Cone Angle=120
    Cone Axis=<0 0 1>
    Module.UseConeFalloff=false
    Module.UseVelDistribution=false
    Override Seed=false
    Velocity Mode=::NewEnumerator2
    Velocity Speed=400
   ) {module=AddVelocity001}
   &FN_SetSpawnGroup(Number of Spawn Groups=3) {enabled=false}
  ]
  particleUpdateScript [
   &ParticleState()
   &ScaleSpriteSize(
    Scale Factor=&MakeVector2D(
     X=&FloatFromCurve(FloatCurve=DataInterface() {propertyValues {Curve {keys [
          {interpMode=RCIM_Cubic, leaveTangent=11.7691, tangentMode=RCTM_Break, time=0, value=0}
          {arriveTangent=0.0629, interpMode=RCIM_Cubic, leaveTangent=0.0629, tangentMode=RCTM_User, time=0.3, value=1}
          {interpMode=RCIM_Cubic, leaveTangent=0.027, tangentMode=RCTM_Break, time=0.35, value=1}
          {arriveTangent=0.0384, tangentMode=RCTM_User, time=1, value=0}
         ]}}})
     Y=&FloatFromCurve(FloatCurve=DataInterface() {propertyValues {Curve {keys [
          {interpMode=RCIM_Cubic, leaveTangent=11.7691, tangentMode=RCTM_Break, time=0, value=0}
          {arriveTangent=0.0629, interpMode=RCIM_Cubic, leaveTangent=0.0629, tangentMode=RCTM_User, time=0.3, value=1}
          {interpMode=RCIM_Cubic, leaveTangent=0.027, tangentMode=RCTM_Break, time=0.35, value=1}
         ]}}})
    )
    Scale Sprite Size Mode=::NewEnumerator2
   )
   &VortexForce(Origin Pull Amount=9619.46, UseOriginOffset=true, Vortex Axis=<0 0 -1>, Vortex Force Amount=4931.6753)
   &GravityForce(Gravity=&RandomRangeVector(Minimum=<0 0 300>, Maximum=<0 0 400>))
   &CurlNoiseForce(Bias Noise=false, Noise Frequency=10, Noise Strength=100, Pan Noise=false) {enabled=false}
   &Drag(
    Drag=&FloatFromCurve(FloatCurve=DataInterface() {propertyValues {Curve {defaultValue=1, keys [<at 0 0.5 RCIM_Cubic>,<at 1 3 RCIM_Cubic>]}}}, Scale Curve=2)
    Use Linear Drag=true
    Use Rotational Drag=true
   )
   &SolveForcesAndVelocity(Clamp Velocity=false, Limit Acceleration=false)
   &ScaleSpriteSizeBySpeed(Max Scale Factor=<1 2>, Velocity Threshold=3500)
   &ScaleSpriteSize(
    Scale Sprite Size Mode=::NewEnumerator0
    Uniform Scale Factor=&FloatFromCurve(
     FloatCurve=DataInterface() {propertyValues {Curve {keys [<at 0.7 1>,<at 1 0>]}}}
     Scale Curve=&NormalizeDistanceRange(
      Start Position:NiagaraPosition=$Engine.Emitter.SimulationPosition
      End Position:NiagaraPosition=$Particles.Position
      Distance:NiagaraFloat=300
      Invert Normalized Range:NiagaraBool=false
     )
    )
   ) {module=ScaleSpriteSize001}
   &FN_Sprite_Attributes(
    Color=&MultiplyLinearColor(Linear Color=$Particles.Color, Scale Factor=<0.025 0.025 0.025 1>)
    Create Color Attribute=true
    Create SpriteSize Attribute=true
    SpriteSize=&Multiply_Vector2D(A=$Particles.SpriteSize, B=<5 2>)
    Write to Module Namespace=true
   )
   &FN_Sprite_Attributes(
    Color=<rgba 0.08 0.0776 0.0791 1>
    Create Color Attribute=true
    Create Position Attribute=true
    Create Sprite Rotation Attribute=true
    Create SpriteSize Attribute=true
    Position:NiagaraPosition=&Multiply_Vector(
     A=Linked() {linkedVariable {name=Particles.Position, type=NiagaraPosition}}
     B=&RandomRangeVector(Minimum=<1.1 1.1 1>, Maximum=<2 2 1>)
    )
    SpriteRotation:NiagaraFloat=&RandomRangeFloat(Minimum=0, Maximum=360)
    SpriteSize=<0.001 0.001>
    Write to Module Namespace=true
   ) {module=FN_Sprite_Attributes001}
   &ScaleColor(
    Scale Alpha=&FloatFromCurve(FloatCurve=DataInterface() {propertyValues {Curve {
        defaultValue=1
        keys [
         {arriveTangent=9.0725, interpMode=RCIM_Cubic, leaveTangent=9.0725, tangentMode=RCTM_User, time=0, value=0}
         {arriveTangent=-2.5249, interpMode=RCIM_Cubic, leaveTangent=-2.5249, tangentMode=RCTM_User, time=0.1408, value=0.9847}
         {arriveTangent=-0.0405, interpMode=RCIM_Cubic, leaveTangent=-0.0405, tangentMode=RCTM_User, time=1, value=0}
        ]
       }}})
    ScaleA=true
    ScaleRGB=true
   )
   &OffsetPosition(Position Offset=&MakeVector(
     X=0
     Y=0
     Z=&Waveform(Global Phase=$Particles.Age, "Waveform [1]"=::NewEnumerator0, "[1] Frequency"=2, "[1] Amplitude Min/Max"=<0 1>, "[1] Scale"=20)
    )) {enabled=false}
   &CameraOffset(Camera Offset Amount=15)
  ]
  rendererClasses [&NiagaraSpriteRendererProperties]
  renderers [
   {rendererClass=&NiagaraSpriteRendererProperties}
   {rendererClass=&NiagaraSpriteRendererProperties, rendererIndex=1}
  ]
 }
 NE_TopRings003 {
  emitterUpdateScript [
   &EmitterState(
    Inactive Response=::NewEnumerator1
    Life Cycle Mode=::NewEnumerator1
    Loop Behavior=::NewEnumerator1
    Loop Delay=$User.ring_delay
    Loop Duration=$User.ring life
    UseLoopDelay=true
   )
   &SpawnBurst_Instantaneous(Spawn Count=5, Use Spawn Probability=false, UseLoopCountLimit=false)
   &SpawnRate(SpawnRate=10, Use Spawn Probability=false)
  ]
  particleSpawnScript [
   /SetVariables(Particles.Distribution:NiagaraFloat=&RandomRangeFloat(Minimum=0, Maximum=0)) {isSetParameters=true}
   &InitializeParticle(
    Color Mode=::NewEnumerator0
    Lifetime=2
    Mass Mode=::NewEnumerator1
    Mesh Scale=&Multiply_VectorByFloat(
     Vector=&RandomRangeVector(Minimum=<1.3 1.3 1.62>, Maximum=<1.3 1.3 1.62>)
     Float=&Multiply_Float(
      A=&RandomRangeFloat(Minimum=1, Maximum=1.1)
      B=&Lerp_Float(A=0.7, B=0.9, Alpha=$Particles.Distribution)
     )
    )
    Mesh Scale Mode=::NewEnumerator4
    Position Offset=&AddVectorToPosition(
     Position=<0 0 0>
     Vector=&MakeVector(X=0, Y=0, Z=&Lerp_Float(A=140, B=200, Alpha=$Particles.Distribution))
    )
    UsePositionOffset=true
    Write Material Random=true
   )
   &OffsetPosition(Position Offset=&RandomRangeVector(Minimum=<-20 -20 0>, Maximum=<20 20 0>)) {module=OffsetPosition001}
   &InitialMeshOrientation(
    Mesh Axis To Orient=&RandomRangeVector(Minimum=<1 1 -0.1>, Maximum=<-1 1 0.1>)
    Mesh Orientation Mode=::NewEnumerator2
    Orientation Vector=<0 1 0>
    Rotation=&RandomRangeVector(Minimum=<0 0 0>, Maximum=<0 0 360>)
    Use Rotation Vector=true
   )
   &FN_SetSpawnGroup(Number of Spawn Groups=2)
  ]
  particleUpdateScript [
   &ParticleState()
   &ScaleMeshSize(Scale Factor=&Multiply_VectorByFloat(
     Vector=&Multiply_VectorByFloat(Vector=$Engine.Owner.Scale, Float=1)
     Float=&FloatFromCurve(FloatCurve=DataInterface() {propertyValues {Curve {keys [
          <at 0 0.2 RCIM_Cubic>
          {arriveTangent=-0.1281, interpMode=RCIM_Cubic, leaveTangent=-0.1281, tangentMode=RCTM_User, time=0.3, value=1}
          {arriveTangent=-0.184, interpMode=RCIM_Cubic, leaveTangent=-0.184, tangentMode=RCTM_User, time=0.5, value=1}
          {arriveTangent=-0.1209, interpMode=RCIM_Cubic, leaveTangent=-0.1209, tangentMode=RCTM_User, time=1, value=1}
         ]}}})
    )) {enabled=false, module=ScaleMeshSize002}
   &ScaleMeshSize(Scale Factor=&Multiply_VectorByFloat(
     Vector=&Multiply_VectorByFloat(Vector=$Engine.Owner.Scale, Float=1)
     Float=&FloatFromCurve(
      FloatCurve=DataInterface() {propertyValues {Curve {keys [
          {arriveTangent=-0.2182, interpMode=RCIM_Cubic, time=0, value=1}
          {arriveTangent=-0.0244, interpMode=RCIM_Cubic, leaveTangent=-0.0244, tangentMode=RCTM_User, time=0.15, value=1}
          <at 0.8 1 RCIM_Cubic>
          <at 1 1 RCIM_Cubic>
         ]}}}
      CurveIndex=$Emitter.NormalizedLoopAge
      Scale Curve=3.5
     )
    )) {module=ScaleMeshSize001}
   &UpdateMeshOrientation(Rotation Rate=0.5, Rotation Vector=&RandomRangeVector(Minimum=<-0.1 -0.1 -1>, Maximum=<0.1 0.1 -1>)) {module=UpdateMeshOrientation001}
   &ShapeLocation(
    MaskSpawnGroup=false
    Offset=&MakeVector(
     X=&MakeFloatFromVector(
      Vector=&MakeVector(
       X=&Sine(
        Normalized Angle=$System.Age
        Period=&RandomRangeFloat(Minimum=1, Maximum=1.5)
        Scale=&RandomRangeFloat(Minimum=10, Maximum=25)
        Bias=0
       )
       Y=&MakeFloatFromVector(
        Vector=&MakeVector(
         X=&Sine(
          Normalized Angle=$System.Age
          Period=&RandomRangeFloat(Minimum=1, Maximum=1.5)
          Scale=&RandomRangeFloat(Minimum=7, Maximum=20)
          Bias=0
         )
         Y=0
         Z=&FloatFromCurve(
          FloatCurve=DataInterface() {propertyValues {Curve {keys [
              <at 0 79 RCIM_Cubic>
              {arriveTangent=9.1812, interpMode=RCIM_Cubic, leaveTangent=9.1812, tangentMode=RCTM_User, time=0.7, value=79}
              {arriveTangent=85.7143, interpMode=RCIM_Cubic, leaveTangent=85.7143, time=0.75, value=46}
              {arriveTangent=279.3773, interpMode=RCIM_Cubic, leaveTangent=279.3773, tangentMode=RCTM_User, time=0.77, value=85}
             ]}}}
          CurveIndex=$Emitter.NormalizedLoopAge
         )
        )
        Channel=::NewEnumerator0
       )
       Z=&FloatFromCurve(
        FloatCurve=DataInterface() {propertyValues {Curve {keys [
            <at 0 79 RCIM_Cubic>
            {arriveTangent=9.1812, interpMode=RCIM_Cubic, leaveTangent=9.1812, tangentMode=RCTM_User, time=0.86, value=79}
            {arriveTangent=60, interpMode=RCIM_Cubic, leaveTangent=60, time=0.91, value=46}
            {arriveTangent=279.3773, interpMode=RCIM_Cubic, leaveTangent=279.3773, tangentMode=RCTM_User, time=0.96, value=85}
           ]}}}
        CurveIndex=$Emitter.NormalizedLoopAge
       )
      )
      Channel=::NewEnumerator0
     )
     Y=0
     Z=&FloatFromCurve(
      FloatCurve=DataInterface() {propertyValues {Curve {keys [
          {arriveTangent=-213.3342, interpMode=RCIM_Cubic, leaveTangent=-213.3342, tangentMode=RCTM_User, time=0.81, value=0}
          {arriveTangent=233.3334, interpMode=RCIM_Cubic, leaveTangent=233.3334, time=0.86, value=-63}
          {arriveTangent=279.3773, interpMode=RCIM_Cubic, leaveTangent=279.3773, tangentMode=RCTM_User, time=0.9, value=21}
         ]}}}
      CurveIndex=$Emitter.NormalizedLoopAge
     )
    )
    Offset Mode=::NewEnumerator0
    Override Seed=false
    Sphere Radius=0
    UseOwnerScale=true
   )
   &ScaleMeshSize(Scale Factor=&Multiply_VectorByFloat(
     Vector=&Multiply_VectorByFloat(Vector=$Engine.Owner.Scale, Float=1)
     Float=&FloatFromCurve(
      FloatCurve=DataInterface() {propertyValues {Curve {keys [
          <at -0.0002 1 RCIM_Cubic>
          <at 0.15 1 RCIM_Cubic>
          {arriveTangent=-1.1765, interpMode=RCIM_Cubic, leaveTangent=-1.1765, time=0.85, value=1}
          <at 1 0 RCIM_Cubic>
         ]}}}
      Scale Curve=3
     )
    )) {enabled=false}
   &UpdateMeshOrientation(Rotation Rate=&RandomRangeFloat(Minimum=-1.1, Maximum=-1.2), Rotation Vector=<0 0 1>)
   &OffsetPosition(Position Offset=&Multiply_VectorByFloat(Vector=$System.InterpolateOverTime.InterpolateOverTime_V3_.Moving Average, Float=0.15))
   &SolveForcesAndVelocity(Clamp Velocity=false, Limit Acceleration=false)
   &FN_Mesh_Attributes(
    Create Position Attribute=true
    Create Scale Attribute=true
    Scale=/Niagara/DynamicInputs/ScaleByCurve/ScaleVectorByCurve(
     InitialValue:Vector3f=<1 1 1>
     ScaleCurve:NiagaraDataInterfaceVectorCurve=DataInterface() {propertyValues {XCurve {keys [<at 0 3>,<at 1 0>]}, YCurve {keys [<at 0 3>,<at 1 0>]}, ZCurve {keys [<at 0.0213 1>]}}}
     CurveIndex:NiagaraFloat=$Particles.NormalizedAge
    )
    Write to Module Namespace=true
   )
   &ScaleColor(
    Scale Alpha=&FloatFromCurve(
     FloatCurve=DataInterface() {propertyValues {Curve {keys [
         <at 0.08 0 RCIM_Cubic>
         {arriveTangent=1.357, interpMode=RCIM_Cubic, leaveTangent=0.079, tangentMode=RCTM_Break, time=0.25, value=0.5}
         {arriveTangent=0.1455, interpMode=RCIM_Cubic, leaveTangent=0.1455, tangentMode=RCTM_User, time=0.8, value=0.5}
         {arriveTangent=-0.0405, interpMode=RCIM_Cubic, leaveTangent=-0.0405, tangentMode=RCTM_User, time=0.895, value=0}
        ]}}}
     CurveIndex=$Emitter.NormalizedLoopAge
     Scale Curve=2
    )
    Scale RGB=&Multiply_VectorByFloat(Vector=<1 1 1>, Float=1)
    ScaleA=true
    ScaleRGB=false
   ) {module=ScaleColor002}
   &DynamicMaterialParameters(
    Index 0 Param 1=&FloatFromCurve(FloatCurve=DataInterface() {propertyValues {Curve {keys [<at 0 50>,<at 0.2 50>,<at 1 1>]}}})
    Index 0 Param 3=&FloatFromCurve(FloatCurve=DataInterface() {propertyValues {Curve {keys [<at 0 1 RCIM_Cubic>, {arriveTangent=-3.3322, tangentMode=RCTM_User, time=1, value=0}]}}})
    Module.Index0Float0Write=true
    Module.Index0Float1Write=true
    Module.Index0Float2Write=true
    Module.Index0Float3Write=true
   )
   &DynamicMaterialParameters(
    Index 1 Param 1=&Sine(
     Normalized Angle=$System.Age
     Period=&RandomRangeFloat(Minimum=1, Maximum=1.5)
     Scale=&RandomRangeFloat(Minimum=30, Maximum=40)
     Bias=0
    )
    Index 1 Param 2=&FloatFromCurve(
     FloatCurve=DataInterface() {propertyValues {Curve {keys [
         <at 0.04 0 RCIM_Cubic>
         {arriveTangent=-0.0868, interpMode=RCIM_Cubic, leaveTangent=-0.0868, tangentMode=RCTM_User, time=0.3, value=0.9}
         {arriveTangent=-0.0201, interpMode=RCIM_Cubic, leaveTangent=-0.0201, tangentMode=RCTM_User, time=0.5, value=0.9}
         {arriveTangent=0.3632, interpMode=RCIM_Cubic, leaveTangent=0.3632, tangentMode=RCTM_User, time=0.7, value=0.9}
         {arriveTangent=-0.3399, interpMode=RCIM_Cubic, leaveTangent=-0.3399, tangentMode=RCTM_User, time=1, value=0}
        ]}}}
     Scale Curve=0.1
    )
    Index 1 Param 3=3
    Index 3 Param 3=&FloatFromCurve(FloatCurve=DataInterface() {propertyValues {Curve {
        defaultValue=1
        keys [
         {interpMode=RCIM_Cubic, tangentMode=RCTM_User, time=0, value=-1}
         {arriveTangent=1.3636, interpMode=RCIM_Cubic, leaveTangent=1.3636, time=0.15, value=0.2}
         {interpMode=RCIM_Cubic, tangentMode=RCTM_User, time=0.88, value=0.2}
         {tangentMode=RCTM_User, time=1, value=-1}
        ]
       }}})
    Module.Index1Float0Write=true
    Module.Index1Float1Write=true
    Module.Index1Float2Write=true
    Module.Index1Float3Write=false
    Module.Index3Float0Write=false
    Module.Index3Float1Write=false
    Module.Index3Float2Write=false
    Module.Index3Float3Write=true
    Write Parameter Index 0=::NewEnumerator0
    Write Parameter Index 1=::NewEnumerator1
    Write Parameter Index 3=::NewEnumerator1
   ) {module=DynamicMaterialParameters001}
   &ScaleColor(
    Scale Alpha=&FloatFromCurve(FloatCurve=DataInterface() {propertyValues {Curve {keys [<at 0 0>,<at 0.0437 1>,<at 0.5 1>,<at 1 0>]}}})
    ScaleA=true
    ScaleRGB=true
   )
   &CameraOffset(Camera Offset Amount=15)
  ]
  rendererClasses [&NiagaraMeshRendererProperties]
  renderers [
   {rendererClass=&NiagaraMeshRendererProperties, rendererIndex=0}
   {rendererClass=&NiagaraMeshRendererProperties, rendererIndex=1}
   {rendererClass=&NiagaraMeshRendererProperties, rendererIndex=2}
   {rendererClass=&NiagaraMeshRendererProperties, rendererIndex=3}
   {rendererClass=&NiagaraMeshRendererProperties, rendererIndex=4}
   {rendererClass=&NiagaraMeshRendererProperties, rendererIndex=5}
  ]
 }
 NE_flameTrail {
  emitterUpdateScript [
   &EmitterState(Inactive Response=::NewEnumerator1, Life Cycle Mode=::NewEnumerator1, Loop Behavior=::NewEnumerator1, Loop Duration=5.4, UseLoopDelay=false)
   &SpawnRate(SpawnRate=0, Use Spawn Probability=false) {enabled=false}
   &SpawnParticlesFromOtherEmitter(
    Attribute Reader=DataInterface() {propertyValues {EmitterBinding {bindingMode=Other, emitterName=NE_FireSprites}}}
    Spawn Rate=1000
    Use Maximum Age=false
    Use Minimum Age=false
   )
   &SpawnParticlesFromOtherEmitter(
    Attribute Reader=DataInterface() {propertyValues {EmitterBinding {bindingMode=Other, emitterName=NE_FireSprites001}}}
    Spawn Rate=1000
    Use Maximum Age=false
    Use Minimum Age=false
   ) {enabled=false, module=SpawnParticlesFromOtherEmitter001}
  ]
  particleSpawnScript [
   &InitializeParticle(
    Lifetime=0.3
    Ribbon Width=&RandomRangeFloat(Minimum=10, Maximum=12)
    Ribbon Width Mode=::NewEnumerator1
    UsePositionOffset=false
    Write Material Random=true
   )
   &SampleParticlesFromOtherEmitter(Color Sampling=::NewEnumerator1, Particle ID Sampling=::NewEnumerator1) {module=SampleParticlesFromOtherEmitter001}
   &SampleParticlesFromOtherEmitter() {enabled=false}
  ]
  particleUpdateScript [
   &SolveForcesAndVelocity(Clamp Velocity=false, Limit Acceleration=false)
   &ParticleState()
   &ScaleRibbonWidth(Ribbon Width Scale=&FloatFromCurve(
     FloatCurve=DataInterface() {propertyValues {Curve {keys [{arriveTangent=-1.1377, interpMode=RCIM_Cubic, leaveTangent=-1.1377, tangentMode=RCTM_User, time=0, value=1}, <at 1 1>]}}}
     Scale Curve=2
    ))
   &ScaleColor(
    Scale Alpha=&FloatFromCurve(FloatCurve=DataInterface() {propertyValues {Curve {keys [{arriveTangent=0.3606, interpMode=RCIM_Cubic, leaveTangent=0.3606, tangentMode=RCTM_User, time=0, value=1}, <at 1 0>]}}})
    ScaleA=true
    ScaleRGB=true
   )
   &self:ReadRibbonSourceParticleAge_06(PAR=DataInterface() {propertyValues {EmitterBinding {bindingMode=Other, emitterName=NE_FireSprites}}}) {module=ReadRibbonSourceParticleAge_06}
   &ScaleColor(
    Scale Alpha=&FloatFromCurve(
     FloatCurve=DataInterface() {propertyValues {Curve {keys [
         {arriveTangent=3.0043, interpMode=RCIM_Cubic, leaveTangent=0.8333, tangentMode=RCTM_Break, time=0.1, value=0.5}
         {arriveTangent=-0.6087, interpMode=RCIM_Cubic, leaveTangent=-0.6087, tangentMode=RCTM_User, time=0.8, value=0.5}
         {arriveTangent=-0.0405, interpMode=RCIM_Cubic, leaveTangent=-0.0405, tangentMode=RCTM_User, time=0.895, value=0}
        ]}}}
     CurveIndex=$Emitter.NormalizedLoopAge
     Scale Curve=2
    )
    Scale RGB=&Multiply_VectorByFloat(Vector=<1 1 1>, Float=1)
    ScaleA=true
    ScaleRGB=false
   ) {module=ScaleColor002}
   &CameraOffset(Camera Offset Amount=15)
  ]
  rendererClasses [&NiagaraRibbonRendererProperties]
  renderers [{rendererClass=&NiagaraRibbonRendererProperties}]
 }
 NE_light {
  emitterUpdateScript [
   &EmitterState(Life Cycle Mode=::NewEnumerator1, Loop Behavior=::NewEnumerator1, Loop Duration=5.6, UseLoopDelay=false)
   &SpawnBurst_Instantaneous(Use Spawn Probability=false, UseLoopCountLimit=false)
  ]
  particleSpawnScript [&InitializeParticle(Color Mode=::NewEnumerator0, Lifetime=5.6, Mesh Scale Mode=::NewEnumerator4, UsePositionOffset=true, Write Material Random=true)]
  particleUpdateScript [
   &ParticleState()
   &FN_Gen5_Light(
    Attenutation:NiagaraFloat=&FloatFromCurve(
     FloatCurve=DataInterface() {propertyValues {Curve {keys [
         <at 0 1 RCIM_Cubic>
         {arriveTangent=104.2105, interpMode=RCIM_Cubic, leaveTangent=104.2105, time=0.07, value=100}
         <at 0.95 100>
         <at 1.0037 0>
        ]}}}
     Scale Curve=1.5
    )
    Color:LinearColor=<rgba 1 0.5418 0.0396 1>
    Intensity:NiagaraFloat=&Multiply_Float(
     A=&FloatFromCurve(
      FloatCurve=DataInterface() {propertyValues {Curve {keys [<at 0 1 RCIM_Cubic>, {arriveTangent=3123.75, interpMode=RCIM_Cubic, leaveTangent=3123.75, time=0.07, value=2500}, <at 0.8 2500>, <at 0.95 0>]}}}
      Scale Curve=0.5
     )
     B=&ScaleAndBiasFloat(Float:NiagaraFloat=$System.Flicker, Scale:NiagaraFloat=2, Bias:NiagaraFloat=0.5)
    )
    Position Offset:Vector3f=<0 0 15>
   )
   &SolveForcesAndVelocity(Clamp Velocity=false, Limit Acceleration=false)
   &CameraOffset(Camera Offset Amount=15)
  ]
  rendererClasses [&NiagaraComponentRendererProperties]
  renderers [{rendererClass=&NiagaraComponentRendererProperties, rendererIndex=0}]
 }
 NE_tornado {
  emitterUpdateScript [
   &EmitterState(Life Cycle Mode=::NewEnumerator1, Loop Behavior=::NewEnumerator1, Loop Duration=6.2, UseLoopDelay=false)
   &SpawnBurst_Instantaneous(Use Spawn Probability=false, UseLoopCountLimit=false)
  ]
  particleSpawnScript [&InitializeParticle(Color Mode=::NewEnumerator0, Lifetime=5.6, Mesh Scale Mode=::NewEnumerator4, UsePositionOffset=true, Write Material Random=true)]
  particleUpdateScript [
   &DynamicMaterialParameters(
    Index 0 Linear Color:LinearColor=&ColorFromCurve(ColorCurve=DataInterface() {propertyValues {
       AlphaCurve {keys [<at 0 1 RCIM_Cubic>,<at 1 0.34 RCIM_Cubic>]}
       BlueCurve {keys [<at 0.0272 1>,<at 0.9588 0.9066>]}
       GreenCurve {keys [
         {arriveTangent=0.3437, time=-0.2, value=0.7}
         {arriveTangent=0.0417, interpMode=RCIM_Cubic, leaveTangent=0.0417, time=0.05, value=0.7}
         <at 0.1 0.7125 RCIM_Cubic>
         {arriveTangent=0.0032, interpMode=RCIM_Cubic, leaveTangent=0.0032, tangentMode=RCTM_User, time=0.1295, value=0.7}
         <at 1 0.7>
         {arriveTangent=0.2562, leaveTangent=0.2562, time=1, value=0.7}
        ]}
       RedCurve {keys [<at 0.0272 1>,<at 0.3305 1 RCIM_Cubic>,<at 0.4782 1 RCIM_Cubic>,<at 0.5373 1>,<at 0.9588 1>]}
      }})
    Write Parameter Index 0=::NewEnumerator5
   )
   &DynamicMaterialParameters(
    Index 0 Linear Color:LinearColor=&ColorFromCurve(ColorCurve=DataInterface() {propertyValues {
       AlphaCurve {keys [<at 0 1 RCIM_Cubic>,<at 1 0.34 RCIM_Cubic>]}
       BlueCurve {keys [<at 0.0272 1>,<at 0.9588 0.9066>]}
       GreenCurve {keys [
         {arriveTangent=0.3437, interpMode=RCIM_Cubic, leaveTangent=0.3437, tangentMode=RCTM_User, time=0, value=0.7}
         <at 1 0.7>
         {arriveTangent=0.2562, interpMode=RCIM_Cubic, leaveTangent=0.2562, tangentMode=RCTM_User, time=1, value=0.7}
        ]}
       RedCurve {keys [<at 0.0272 1>,<at 0.3305 1 RCIM_Cubic>,<at 0.4782 1 RCIM_Cubic>,<at 0.5373 1>,<at 0.9588 1>]}
      }})
    Index 1 Param 1=&MakeFloatFromVector(Vector=$System.InterpolateOverTime.InterpolateOverTime_V3_.Moving Average, Channel=::NewEnumerator0)
    Index 1 Param 2=&MakeFloatFromVector(Vector=$System.InterpolateOverTime.InterpolateOverTime_V3_.Moving Average, Channel=::NewEnumerator1)
    Index 1 Param 3=&MakeFloatFromVector(Vector=$System.InterpolateOverTime.InterpolateOverTime_V3_.Moving Average, Channel=::NewEnumerator2)
    Module.Index1Float0Write=true
    Module.Index1Float1Write=false
    Module.Index1Float2Write=false
    Module.Index1Float3Write=false
    Write Parameter Index 0=::NewEnumerator5
    Write Parameter Index 1=::NewEnumerator1
   ) {enabled=false, module=DynamicMaterialParameters002}
   &DynamicMaterialParameters(
    Index 1 Param 1=&Sine(
     Normalized Angle=$System.Age
     Period=&RandomRangeFloat(Minimum=1, Maximum=1.5)
     Scale=&RandomRangeFloat(Minimum=30, Maximum=40)
     Bias=0
    )
    Index 1 Param 2=&FloatFromCurve(FloatCurve=DataInterface() {propertyValues {Curve {keys [
         <at 0.04 0 RCIM_Cubic>
         {arriveTangent=-0.0868, interpMode=RCIM_Cubic, leaveTangent=-0.0868, tangentMode=RCTM_User, time=0.1, value=0.9}
         {arriveTangent=-0.0201, interpMode=RCIM_Cubic, leaveTangent=-0.0201, tangentMode=RCTM_User, time=0.5, value=0.9}
         {arriveTangent=0.3632, interpMode=RCIM_Cubic, leaveTangent=0.3632, tangentMode=RCTM_User, time=0.9, value=0.9}
         {arriveTangent=-0.3399, interpMode=RCIM_Cubic, leaveTangent=-0.3399, tangentMode=RCTM_User, time=1, value=0}
        ]}}})
    Index 1 Param 3=1
    Index 3 Param 3=&FloatFromCurve(FloatCurve=DataInterface() {propertyValues {Curve {
        defaultValue=1
        keys [
         {interpMode=RCIM_Cubic, tangentMode=RCTM_User, time=0, value=-1}
         {arriveTangent=1.3636, interpMode=RCIM_Cubic, leaveTangent=1.3636, time=0.15, value=0.2}
         {interpMode=RCIM_Cubic, tangentMode=RCTM_User, time=0.88, value=0.2}
         {tangentMode=RCTM_User, time=1, value=-1}
        ]
       }}})
    Module.Index1Float0Write=true
    Module.Index1Float1Write=true
    Module.Index1Float2Write=true
    Module.Index1Float3Write=false
    Module.Index3Float0Write=false
    Module.Index3Float1Write=false
    Module.Index3Float2Write=false
    Module.Index3Float3Write=true
    Write Parameter Index 0=::NewEnumerator0
    Write Parameter Index 1=::NewEnumerator1
    Write Parameter Index 3=::NewEnumerator1
   ) {module=DynamicMaterialParameters001}
   &ParticleState()
   &UpdateMeshOrientation(
    Facing Direction=&MakeVector(X=0, Y=90, Z=5)
    Facing Interpolation=true
    Orientation Method=::NewEnumerator1
    Override Quat=false
    TransformFacing=false
    TransformUp=false
    Up Direction:Vector3f=&MakeVector(X=0, Y=0, Z=1)
   )
   &ScaleMeshSize(Scale Factor=&MakeVector(X=$User.size, Y=$User.size, Z=&FloatFromCurve(FloatCurve=$User.Curve for Floats)))
   &FN_Gen5_Light(
    Attenutation:NiagaraFloat=&FloatFromCurve(FloatCurve=DataInterface() {propertyValues {Curve {keys [
         <at 0 1 RCIM_Cubic>
         {arriveTangent=104.2105, interpMode=RCIM_Cubic, leaveTangent=104.2105, time=0.07, value=100}
         <at 0.95 100>
         <at 1.0037 0>
        ]}}})
    Color:LinearColor=<rgba 1 0.5418 0.0396 1>
    Intensity:NiagaraFloat=&Multiply_Float(
     A=&FloatFromCurve(
      FloatCurve=DataInterface() {propertyValues {Curve {keys [<at 0 1 RCIM_Cubic>, {arriveTangent=3123.75, interpMode=RCIM_Cubic, leaveTangent=3123.75, time=0.07, value=2500}, <at 0.8 2500>, <at 0.9 0>]}}}
      Scale Curve=0.3
     )
     B=&ScaleAndBiasFloat(Float:NiagaraFloat=$System.Flicker, Scale:NiagaraFloat=1, Bias:NiagaraFloat=0.5)
    )
    Position Offset:Vector3f=<0 0 110>
   )
   &ScaleColor(
    Scale Alpha=&FloatFromCurve(FloatCurve=DataInterface() {propertyValues {Curve {
        defaultValue=1
        keys [
         {interpMode=RCIM_Cubic, tangentMode=RCTM_User, time=0, value=0}
         {arriveTangent=1.1364, interpMode=RCIM_Cubic, leaveTangent=1.1364, time=0.11, value=1}
         {interpMode=RCIM_Cubic, leaveTangent=-32.0356, tangentMode=RCTM_Break, time=0.88, value=1}
         {tangentMode=RCTM_User, time=1, value=0}
        ]
       }}})
    Scale RGB=&VectorFromFloat(Value=&FloatFromCurve(FloatCurve=DataInterface() {propertyValues {Curve {
         defaultValue=1
         keys [
          {interpMode=RCIM_Cubic, tangentMode=RCTM_User, time=0, value=0}
          {arriveTangent=1.1364, interpMode=RCIM_Cubic, leaveTangent=1.1364, time=0.11, value=1}
          {interpMode=RCIM_Cubic, leaveTangent=-32.0356, tangentMode=RCTM_Break, time=0.88, value=1}
          {tangentMode=RCTM_User, time=1, value=0}
         ]
        }}}))
    ScaleA=true
    ScaleRGB=true
   )
   &FN_Mesh_Attributes(
    Color=&ColorFromCurve(ColorCurve=DataInterface() {propertyValues {
       AlphaCurve {keys [<at 0 1 RCIM_Cubic>,<at 0.9972 0 RCIM_Cubic>]}
       BlueCurve {keys [<at 0 1 RCIM_Cubic>,<at 1 1 RCIM_Cubic>]}
       GreenCurve {keys [<at 0 1 RCIM_Cubic>,<at 1 1 RCIM_Cubic>]}
       RedCurve {keys [<at 0 1 RCIM_Cubic>,<at 1 1 RCIM_Cubic>]}
      }})
    Create Color Attribute=true
    "Create Normalized Age "=true
    Create Visibility Attribute=true
    Lifetime=5
    Write to Module Namespace=true
   ) {module=FN_Mesh_Attributes001}
   &ScaleColor(Scale Alpha=$System.TOD, Scale RGB=&VectorFromFloat(Value=$System.TOD), ScaleA=true, ScaleRGB=true) {enabled=false, module=ScaleColor001}
   &MeshRotationForce(Rotation=&MakeVector(
     X=0
     Y=0
     Z=&FloatFromCurve(FloatCurve=DataInterface() {propertyValues {Curve {keys [<at 0 5>,<at 0.2 0>,<at 0.8 0>,<at 1 5>]}}}, Scale Curve=10)
    )) {enabled=false}
   &AddRotationalVelocity(Rotation Rate=&MakeVector(
     X=0
     Y=0
     Z=&FloatFromCurve(
      FloatCurve=DataInterface() {propertyValues {Curve {keys [
          <at 0 -5 RCIM_Cubic>
          {arriveTangent=-0.8681, interpMode=RCIM_Cubic, leaveTangent=-0.8681, tangentMode=RCTM_User, time=0.2, value=0}
          {arriveTangent=-2.3352, interpMode=RCIM_Cubic, leaveTangent=-2.3352, tangentMode=RCTM_User, time=0.8, value=0}
          <at 1 -15 RCIM_Cubic>
         ]}}}
      Scale Curve=0.04
     )
    ))
   &Drag(
    Rotational Drag:NiagaraFloat=&FloatFromCurve(FloatCurve=DataInterface() {propertyValues {Curve {keys [
         <at 0 0 RCIM_Cubic>
         {arriveTangent=5.4945, interpMode=RCIM_Cubic, leaveTangent=5.4945, time=0.1, value=5}
         {arriveTangent=-5.5556, interpMode=RCIM_Cubic, leaveTangent=-5.5556, time=0.91, value=5}
         <at 1 0 RCIM_Cubic>
        ]}}})
    Use Linear Drag=true
    Use Rotational Drag=true
   )
   &SolveForcesAndVelocity(Clamp Velocity=false, Limit Acceleration=false)
   &CameraOffset(Camera Offset Amount=15)
  ]
  rendererClasses [&NiagaraComponentRendererProperties, &NiagaraMeshRendererProperties]
  renderers [
   {rendererClass=&NiagaraComponentRendererProperties, rendererIndex=0}
   {rendererClass=&NiagaraMeshRendererProperties, rendererIndex=1}
   {rendererClass=&NiagaraMeshRendererProperties, rendererIndex=2}
   {rendererClass=&NiagaraMeshRendererProperties, rendererIndex=3}
   {rendererClass=&NiagaraMeshRendererProperties, rendererIndex=4}
   {rendererClass=&NiagaraMeshRendererProperties, rendererIndex=5}
   {rendererClass=&NiagaraMeshRendererProperties, rendererIndex=6}
   {rendererClass=&NiagaraMeshRendererProperties, rendererIndex=7}
  ]
 }
 Ring_Energy {
  emitterUpdateScript [
   &EmitterState(
    Inactive Response=::NewEnumerator1
    Life Cycle Mode=::NewEnumerator1
    Loop Behavior=::NewEnumerator1
    Loop Delay=$User.ring_delay
    Loop Duration=$User.ring life
    UseLoopDelay=true
   )
   &SpawnBurst_Instantaneous(Spawn Count=2, Use Spawn Probability=false, UseLoopCountLimit=false) {enabled=false}
   &SpawnRate(SpawnRate=10, Use Spawn Probability=false)
   &SpawnBurst_Instantaneous(Spawn Count=2, Spawn Time=3.85, Use Spawn Probability=false, UseLoopCountLimit=false) {enabled=false, module=SpawnBurst_Instantaneous001}
  ]
  particleSpawnScript [
   &InitializeParticle(
    Lifetime Max=1
    Lifetime Min=0.7
    Lifetime Mode=::NewEnumerator1
    Mesh Scale=&Multiply_VectorByFloat(
     Vector=&MakeVector(X=0.925, Y=0.925, Z=&RandomRangeFloat(Minimum=1.4, Maximum=2))
     Float=&RandomRangeFloat(Minimum=0.3, Maximum=0.4)
    )
    Mesh Scale Mode=::NewEnumerator4
    UsePositionOffset=true
    Write Material Random=true
   ) {module=InitializeParticle001}
   &InitialMeshOrientation(
    Mesh Orientation Mode=::NewEnumerator2
    Orientation Coordinate Space=::Local
    Rotation=&RandomRangeVector(Minimum=<0 0 -2>, Maximum=<0 0 -5>)
    Use Rotation Vector=true
   ) {module=InitialMeshOrientation001}
   &AddVelocity(Velocity=&RandomRangeVector(Minimum=<0 0 0>, Maximum=<0 0 12>))
  ]
  particleUpdateScript [
   &SolveForcesAndVelocity(Clamp Velocity=false, Limit Acceleration=false)
   &SystemLocation(
    MaskSpawnGroup=false
    Offset=&MakeVector(
     X=&MakeFloatFromVector(
      Vector=&MakeVector(
       X=&Sine(
        Normalized Angle=$System.Age
        Period=&RandomRangeFloat(Minimum=1, Maximum=1.5)
        Scale=&RandomRangeFloat(Minimum=10, Maximum=20)
        Bias=0
       )
       Y=&MakeFloatFromVector(
        Vector=&MakeVector(
         X=&Sine(
          Normalized Angle=$System.Age
          Period=&RandomRangeFloat(Minimum=1, Maximum=1.5)
          Scale=&RandomRangeFloat(Minimum=7, Maximum=20)
          Bias=0
         )
         Y=0
         Z=&FloatFromCurve(
          FloatCurve=DataInterface() {propertyValues {Curve {keys [
              <at 0 79 RCIM_Cubic>
              {arriveTangent=9.1812, interpMode=RCIM_Cubic, leaveTangent=9.1812, tangentMode=RCTM_User, time=0.7, value=79}
              {arriveTangent=85.7143, interpMode=RCIM_Cubic, leaveTangent=85.7143, time=0.75, value=46}
              {arriveTangent=279.3773, interpMode=RCIM_Cubic, leaveTangent=279.3773, tangentMode=RCTM_User, time=0.77, value=85}
             ]}}}
          CurveIndex=$Emitter.NormalizedLoopAge
         )
        )
        Channel=::NewEnumerator0
       )
       Z=&FloatFromCurve(
        FloatCurve=DataInterface() {propertyValues {Curve {keys [
            <at 0 79 RCIM_Cubic>
            {arriveTangent=9.1812, interpMode=RCIM_Cubic, leaveTangent=9.1812, tangentMode=RCTM_User, time=0.86, value=79}
            {arriveTangent=60, interpMode=RCIM_Cubic, leaveTangent=60, time=0.91, value=46}
            {arriveTangent=279.3773, interpMode=RCIM_Cubic, leaveTangent=279.3773, tangentMode=RCTM_User, time=0.96, value=85}
           ]}}}
        CurveIndex=$Emitter.NormalizedLoopAge
       )
      )
      Channel=::NewEnumerator0
     )
     Y=0
     Z=&FloatFromCurve(
      FloatCurve=DataInterface() {propertyValues {Curve {keys [
          {arriveTangent=9.1812, interpMode=RCIM_Cubic, leaveTangent=9.1812, tangentMode=RCTM_User, time=0.81, value=125}
          {arriveTangent=166.6667, interpMode=RCIM_Cubic, leaveTangent=166.6667, time=0.86, value=77}
          {arriveTangent=279.3773, interpMode=RCIM_Cubic, leaveTangent=279.3773, tangentMode=RCTM_User, time=0.9, value=140}
         ]}}}
      CurveIndex=$Emitter.NormalizedLoopAge
     )
    )
   )
   &ParticleState()
   &UpdateMeshOrientation(
    Rotation Rate=&SelectFloatFromArray(
     Array Sampling Mode=::NewEnumerator0
     Direct Array Index=?
     Float Selection Array=DataInterface() {propertyValues {FloatData [0.3,0.3,0.2,0.2]}}
     Random Seed=0
    )
    Rotation Vector=<0 0 -1>
   )
   &DynamicMaterialParameters(
    Index 0 Param 1=&FloatFromCurve(
     FloatCurve=DataInterface() {propertyValues {Curve {keys [{interpMode=RCIM_Cubic, leaveTangent=-0.0324, tangentMode=RCTM_Break, time=0, value=1.5}, <at 1 -0.5>]}}}
     Scale Curve=&RandomRangeFloat(Minimum=0.8, Maximum=1)
    )
    Index 0 Param 2=&FloatFromCurve(FloatCurve=DataInterface() {propertyValues {Curve {keys [
         {arriveTangent=0.3262, interpMode=RCIM_Cubic, leaveTangent=0.3262, tangentMode=RCTM_User, time=0.2, value=0}
         {arriveTangent=1.6237, interpMode=RCIM_Cubic, leaveTangent=1.6237, tangentMode=RCTM_User, time=1, value=1}
        ]}}})
    Module.Index0Float0Write=true
    Module.Index0Float1Write=true
    Module.Index0Float2Write=true
    Module.Index0Float3Write=true
   )
   &ScaleColor(
    Scale Alpha=&FloatFromCurve(
     FloatCurve=DataInterface() {propertyValues {Curve {keys [
         <at 0.09 0 RCIM_Cubic>
         {arriveTangent=1.357, interpMode=RCIM_Cubic, leaveTangent=0.079, tangentMode=RCTM_Break, time=0.35, value=0.5}
         {arriveTangent=0.1455, interpMode=RCIM_Cubic, leaveTangent=0.1455, tangentMode=RCTM_User, time=0.8, value=0.5}
         {arriveTangent=-0.0405, interpMode=RCIM_Cubic, leaveTangent=-0.0405, tangentMode=RCTM_User, time=0.895, value=0}
        ]}}}
     CurveIndex=$Emitter.NormalizedLoopAge
     Scale Curve=2
    )
    Scale RGB=&Multiply_VectorByFloat(Vector=<1 1 1>, Float=1)
    ScaleA=true
    ScaleRGB=false
   ) {module=ScaleColor002}
   &ScaleMeshSize(Scale Factor=&Multiply_VectorByFloat(
     Vector=&Multiply_VectorByFloat(
      Vector=&MakeVector(
       X=1
       Y=1
       Z=&Multiply_Float(
        A=&FloatFromCurve(
         FloatCurve=DataInterface() {propertyValues {Curve {keys [{arriveTangent=-1.572, time=0, value=1}, {interpMode=RCIM_Cubic, tangentMode=RCTM_User, time=1, value=0.5}]}}}
         Scale Curve=0.5
        )
        B=&RandomRangeFloat(Minimum=1, Maximum=1)
       )
      )
      Float=0.8
     )
     Float=&FloatFromCurve(
      FloatCurve=DataInterface() {propertyValues {Curve {keys [
          {arriveTangent=4.7736, interpMode=RCIM_Cubic, leaveTangent=4.7736, tangentMode=RCTM_User, time=0, value=2.0364}
          {arriveTangent=0.9183, interpMode=RCIM_Cubic, leaveTangent=0.9183, tangentMode=RCTM_User, time=1, value=3.5}
         ]}}}
      Scale Curve=&Multiply_Float(
       A=1.7
       B=&FloatFromCurve(
        FloatCurve=DataInterface() {propertyValues {Curve {keys [{interpMode=RCIM_Cubic, tangentMode=RCTM_User, time=0, value=0.4}, {interpMode=RCIM_Cubic, tangentMode=RCTM_User, time=0.2095, value=1}]}}}
        CurveIndex=$Emitter.Age
       )
      )
     )
    ))
   &ScaleColor(
    Scale Alpha=&FloatFromCurve(
     FloatCurve=DataInterface() {propertyValues {Curve {keys [
         {arriveTangent=-0.0329, interpMode=RCIM_Cubic, leaveTangent=-0.0329, tangentMode=RCTM_User, time=0, value=0}
         {arriveTangent=3.3381, interpMode=RCIM_Cubic, leaveTangent=3.3381, tangentMode=RCTM_User, time=0.1, value=0.6}
         {arriveTangent=0.0085, interpMode=RCIM_Cubic, leaveTangent=0.0085, tangentMode=RCTM_User, time=0.2, value=0.75}
         {arriveTangent=-1.463, interpMode=RCIM_Cubic, leaveTangent=-1.463, tangentMode=RCTM_User, time=0.6, value=0.45}
         {arriveTangent=0.0414, interpMode=RCIM_Cubic, leaveTangent=0.0414, tangentMode=RCTM_User, time=1, value=0}
        ]}}}
     Scale Curve=&RandomRangeFloat(Minimum=1, Maximum=1.4)
    )
    Scale RGB=&MakeVectorFromLinearColorRGB(Color=&LerpLinearColors(
      StartColor=&SelectLinearColorFromArray(
       Array Sampling Mode=::NewEnumerator0
       Direct Array Index=?
       Color Selection Array=DataInterface() {propertyValues {ColorData [<rgba 2 0.3573 0 1>, <rgba 2 0.7011 0 1>]}}
       Random Seed=0
      )
      EndColor=<rgba 20 7.1973 0 1>
      LerpFactor=&FloatFromCurve(
       FloatCurve=DataInterface() {propertyValues {Curve {keys [
           {arriveTangent=-7.9858, interpMode=RCIM_Cubic, leaveTangent=-7.9858, tangentMode=RCTM_User, time=0, value=1}
           {interpMode=RCIM_Cubic, tangentMode=RCTM_User, time=0.4067, value=0}
          ]}}}
       CurveIndex=$Emitter.Age
      )
     ))
    ScaleA=true
    ScaleRGB=true
   )
   &CameraOffset(Camera Offset Amount=15)
  ]
  rendererClasses [&NiagaraMeshRendererProperties]
  renderers [{rendererClass=&NiagaraMeshRendererProperties, rendererIndex=0}]
 }
}
summary {
 systemName=NS_Emote_BlessFlan_FireTornado1
 emitters [
  {emitterName=NE_tornado, rendererClasses [&NiagaraComponentRendererProperties, &NiagaraMeshRendererProperties]}
  {
   emitterName=NE_BottomBase
   rendererClasses [&NiagaraMeshRendererProperties, &NiagaraComponentRendererProperties]
  }
  {emitterName=NE_Sparks, simTarget=GPUComputeSim, rendererClasses [&NiagaraSpriteRendererProperties]}
  {emitterName=NE_ParticleDecalFlash, rendererClasses [&NiagaraSpriteRendererProperties]}
  {emitterName=NE_Glow_Body, rendererClasses [&NiagaraMeshRendererProperties]}
  {emitterName=NE_flameTrail, rendererClasses [&NiagaraRibbonRendererProperties]}
  {emitterName=NE_FireSprites, rendererClasses [&NiagaraSpriteRendererProperties]}
  {emitterName=Ring_Energy, rendererClasses [&NiagaraMeshRendererProperties]}
  {emitterName=NE_TopRings003, rendererClasses [&NiagaraMeshRendererProperties]}
  {emitterName=NE_light, rendererClasses [&NiagaraComponentRendererProperties]}
 ]
 userVariables [
  User.bIsFrontEnd {type=NiagaraBool}
  User.bIsFrontEndPreview {type=NiagaraBool}
  User.Curl Noise {
   type=NiagaraDataInterfaceCurlNoise
   description=""
   defaultValue {
    struct=NiagaraExt_VariableValue_DataInterface
    value {
     dataInterface=&self:NiagaraDataInterfaceCurlNoise_0
     dataInterfaceClass=&NiagaraDataInterfaceCurlNoise
    }
   }
  }
  User.Curve for Colors {
   type=NiagaraDataInterfaceColorCurve
   description=""
   defaultValue {
    struct=NiagaraExt_VariableValue_DataInterface
    value {
     dataInterface=&self:NiagaraDataInterfaceColorCurve_0
     dataInterfaceClass=&NiagaraDataInterfaceColorCurve
    }
   }
  }
  User.Curve for Floats {
   type=NiagaraDataInterfaceCurve
   description=""
   defaultValue {
    struct=NiagaraExt_VariableValue_DataInterface
    value {
     dataInterface=&self:NiagaraDataInterfaceCurve_0
     dataInterfaceClass=&NiagaraDataInterfaceCurve
    }
   }
  }
  User.On_BuiltIn {type=NiagaraFloat, defaultValue=0, description=""}
  User.Position {type=NiagaraPosition, description="", defaultValue {}}
  User.ring life {type=NiagaraFloat, defaultValue=5, description=""}
  User.ring_delay {type=NiagaraFloat, defaultValue=0.6, description=""}
  User.size {type=NiagaraFloat, defaultValue=1, description=""}
  User.WPO_Base {type=LinearColor, description="", defaultValue {struct=LinearColor, value=<rgba 1 1 1 1>}}
 ]
}
